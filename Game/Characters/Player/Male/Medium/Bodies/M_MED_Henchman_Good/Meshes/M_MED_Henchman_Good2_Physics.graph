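# Physics asset for the Henchman Good male medium player body, simulating its coat with front, side and back panels on each side.
# Pelvis and thigh capsules form the lower body, with six tiny master capsules joined to the pelvis by angularly free constraints.
# Each master drives a chain of two flat boxes with per-panel swing and twist limits.
# Extra angularly free constraints tie the lower front and side boxes together on each side.
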PhysicsAsset /Game/Characters/Player/Male/Medium/Bodies/M_MED_Henchman_Good/Meshes/M_MED_Henchman_Good2_Physics
bodies {
 dyn_coat_MSTR_back_l [{shapeType=Capsule, length=0.505, radius=0.505}]
 dyn_coat_MSTR_back_r [{shapeType=Capsule, length=0.505, radius=0.505}]
 dyn_coat_MSTR_frnt_l [{shapeType=Capsule, length=0.505, radius=0.505}]
 dyn_coat_MSTR_frnt_r [{shapeType=Capsule, length=0.505, radius=0.505}]
 dyn_coat_MSTR_side_l [{shapeType=Capsule, length=0.505, radius=0.505}]
 dyn_coat_MSTR_side_r [{shapeType=Capsule, length=0.505, radius=0.505}]
 dyn_coat_back_01_l [{shapeType=Box, extentX=10, extentY=2, extentZ=15, center=<5 0 0>}]
 dyn_coat_back_01_r [{shapeType=Box, extentX=10, extentY=2, extentZ=15, center=<5 0 0>}]
 dyn_coat_back_l [{shapeType=Box, extentX=10, extentY=2, extentZ=15, center=<5 0 0>}]
 dyn_coat_back_r [{shapeType=Box, extentX=10, extentY=2.2067, extentZ=15, center=<5 0 0>}]
 dyn_coat_frnt_01_l [{shapeType=Box, extentX=10, extentY=2.1166, extentZ=13.8661, center=<5 0 0>}]
 dyn_coat_frnt_01_r [{shapeType=Box, extentX=10, extentY=1.9692, extentZ=14.9906, center=<5 0.399 -0.0846>, rotation=<rot -0.2116 2.1336 3.4469>}]
 dyn_coat_frnt_l [{shapeType=Box, extentX=10, extentY=1.445, extentZ=15, center=<5 0 0>}]
 dyn_coat_frnt_r [{shapeType=Box, extentX=10, extentY=2, extentZ=15, center=<5 0 0>}]
 dyn_coat_side_01_l [{shapeType=Box, extentX=10, extentY=2.3397, extentZ=16.7915, center=<5 -0.1701 -0.0272>, rotation=<rot 0.0117 -0.0008 -1.7225>}]
 dyn_coat_side_01_r [{shapeType=Box, extentX=10, extentY=2.753, extentZ=15, center=<5 0 0>}]
 dyn_coat_side_l [{shapeType=Box, extentX=10, extentY=2, extentZ=16.5957, center=<5 -0.939 -0.0594>, rotation=<rot -0.0002 -0.0009 -3.6259>}]
 dyn_coat_side_r [{shapeType=Box, extentX=10, extentY=2, extentZ=15, center=<5 0 0>}]
 pelvis [{shapeType=Capsule, length=9.8495, radius=9.8243, center=<-0.0062 -0.8187 -0.0691>}]
 thigh_l [{shapeType=Capsule, length=17.2926, radius=9.2863, center=<-9.9949 -0.8043 0.2826>, rotation=<rot 0 89.9999 -88.5886>}]
 thigh_r [{shapeType=Capsule, length=13.0671, radius=9.4464, center=<7.7267 0.4487 -1.1981>, rotation=<rot 0 87.9613 -89.9999>}]
}
constraints [
 {bone1Name=thigh_l, bone2Name=pelvis, swing1LimitDegrees=45, swing2LimitDegrees=45, twistLimitDegrees=45}
 {bone1Name=thigh_r, bone2Name=pelvis, swing1LimitDegrees=45, swing2LimitDegrees=45, twistLimitDegrees=45}
 {
  bone1Name=dyn_coat_MSTR_frnt_r
  bone2Name=pelvis
  swing1LimitDegrees=45
  swing1Motion=Free
  swing2LimitDegrees=45
  swing2Motion=Free
  twistLimitDegrees=45
  twistMotion=Free
 }
 {
  bone1Name=dyn_coat_MSTR_frnt_l
  bone2Name=pelvis
  swing1LimitDegrees=45
  swing1Motion=Free
  swing2LimitDegrees=45
  swing2Motion=Free
  twistLimitDegrees=45
  twistMotion=Free
 }
 {
  bone1Name=dyn_coat_MSTR_back_l
  bone2Name=pelvis
  swing1LimitDegrees=45
  swing1Motion=Free
  swing2LimitDegrees=45
  swing2Motion=Free
  twistLimitDegrees=45
  twistMotion=Free
 }
 {
  bone1Name=dyn_coat_MSTR_back_r
  bone2Name=pelvis
  swing1LimitDegrees=45
  swing1Motion=Free
  swing2LimitDegrees=45
  swing2Motion=Free
  twistLimitDegrees=45
  twistMotion=Free
 }
 {
  bone1Name=dyn_coat_MSTR_side_r
  bone2Name=pelvis
  swing1LimitDegrees=45
  swing1Motion=Free
  swing2LimitDegrees=45
  swing2Motion=Free
  twistLimitDegrees=45
  twistMotion=Free
 }
 {
  bone1Name=dyn_coat_MSTR_side_l
  bone2Name=pelvis
  swing1LimitDegrees=45
  swing1Motion=Free
  swing2LimitDegrees=45
  swing2Motion=Free
  twistLimitDegrees=45
  twistMotion=Free
 }
 {bone1Name=dyn_coat_frnt_r, bone2Name=dyn_coat_MSTR_frnt_r, swing1LimitDegrees=45, swing2LimitDegrees=30, twistLimitDegrees=5}
 {bone1Name=dyn_coat_frnt_01_r, bone2Name=dyn_coat_frnt_r, swing1LimitDegrees=45, swing2LimitDegrees=10, twistLimitDegrees=5}
 {bone1Name=dyn_coat_frnt_l, bone2Name=dyn_coat_MSTR_frnt_l, swing1LimitDegrees=45, swing2LimitDegrees=25, twistLimitDegrees=10}
 {bone1Name=dyn_coat_frnt_01_l, bone2Name=dyn_coat_frnt_l, swing1LimitDegrees=45, swing2LimitDegrees=25, twistLimitDegrees=10}
 {bone1Name=dyn_coat_back_l, bone2Name=dyn_coat_MSTR_back_l, swing1LimitDegrees=45, swing2LimitDegrees=45, twistLimitDegrees=15}
 {bone1Name=dyn_coat_back_01_l, bone2Name=dyn_coat_back_l, swing1LimitDegrees=45, swing2LimitDegrees=45, twistLimitDegrees=15}
 {bone1Name=dyn_coat_back_r, bone2Name=dyn_coat_MSTR_back_r, swing1LimitDegrees=45, swing2LimitDegrees=45, twistLimitDegrees=15}
 {bone1Name=dyn_coat_back_01_r, bone2Name=dyn_coat_back_r, swing1LimitDegrees=45, swing2LimitDegrees=45, twistLimitDegrees=15}
 {bone1Name=dyn_coat_side_r, bone2Name=dyn_coat_MSTR_side_r, swing1LimitDegrees=45, swing2LimitDegrees=45, twistLimitDegrees=5}
 {bone1Name=dyn_coat_side_01_r, bone2Name=dyn_coat_side_r, swing1LimitDegrees=45, swing2LimitDegrees=45, twistLimitDegrees=5}
 {bone1Name=dyn_coat_side_l, bone2Name=dyn_coat_MSTR_side_l, swing1LimitDegrees=45, swing2LimitDegrees=45, twistLimitDegrees=5}
 {bone1Name=dyn_coat_side_01_l, bone2Name=dyn_coat_side_l, swing1LimitDegrees=45, swing2LimitDegrees=45, twistLimitDegrees=5}
 {
  bone1Name=dyn_coat_side_01_l
  bone2Name=dyn_coat_frnt_01_l
  swing1LimitDegrees=45
  swing1Motion=Free
  swing2LimitDegrees=45
  swing2Motion=Free
  twistLimitDegrees=15
  twistMotion=Free
 }
 {
  bone1Name=dyn_coat_frnt_01_r
  bone2Name=dyn_coat_side_01_r
  swing1LimitDegrees=45
  swing1Motion=Free
  swing2LimitDegrees=45
  swing2Motion=Free
  twistLimitDegrees=45
  twistMotion=Free
 }
]
